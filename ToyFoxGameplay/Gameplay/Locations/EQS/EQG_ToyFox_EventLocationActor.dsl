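; Custom EQS generator for ToyFox that produces the registered event location actors as query items.
; It gets the Athena game state, finds its B_ToyFoxGameStateComponent, and adds each valid actor from GetRegisteredEventLocationActors as a generated item.
; The context actors are not used, and an invalid game state or component yields no items.
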(fn DoItemGenerationFromActors
  (params (ContextActors :type Array%20of%20Actor%20Object%20References))
  (bind _returnvalue (Fortnite|Athena|GameState|GetGameStateAthena))
  (bind _returnvalue_2 (Actor|GetComponentbyClass _returnvalue "/ToyFoxGameplay/Gameplay/B_ToyFoxGameStateComponent.B_ToyFoxGameStateComponent_C"))
  (Utilities|FlowControl|Sequence
    (:then_0
      (Utilities|IsValid _returnvalue
        (:Is%20Valid
          (Utilities|IsValid _returnvalue_2
            (:Is%20Valid
              (for _array_element (Event|GetRegisteredEventLocationActors _returnvalue_2)
                (Utilities|IsValid _array_element
                  (:Is%20Valid
                    (EQS|AddGeneratedActor _array_element)))))))))
    (:then_1
      (return))))
