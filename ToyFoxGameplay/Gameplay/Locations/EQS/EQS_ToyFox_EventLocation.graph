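# Environment query that picks a ToyFox event location actor in Fortnite.
# A custom generator supplies candidate location actors from the querier context, and a filter checks them against an Athena safe zone whose index is bound to a scalable-float data provider.
# A 2D delta-distance test against all players then filters and scores candidates, averaging across players with a negative linear factor so smaller deltas rank higher.
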EnvQuery /ToyFoxGameplay/Gameplay/Locations/EQS/EQS_ToyFox_EventLocation
options [{
  generator {
   name=B_ToyFoxEventLocationActorGenerator_C_0
   class=EQG_ToyFox_EventLocationActor_C
   properties {
    bAutoSortTests=true
    bCanRunAsync=false
    context=&EnvQueryContext_Querier
    envQueryResultNormalizationOption=Default
    generatedItemType=/Script/AIModule.EnvQueryItemType_Actor
    generatorsActionDescription=""
    optionName=""
   }
  }
  tests [
   FortQueryTest_InsideAthenaSafeZone_0 {
    class=FortQueryTest_InsideAthenaSafeZone
    properties {
     bUseCurrentSafeZoneIndicatorRadius=false
     testPurpose=Filter
     safeZoneIndex {
      dataBinding=&self:FortQueryTest_InsideAthenaSafeZone_0.FortAIDataProvider_ScalableFloat_0
      dataField=IntValue
      defaultValue=0
     }
    }
   }
   FortQueryTest_DeltaDistance_0 {
    class=FortQueryTest_DeltaDistance
    properties {
     bDefineReferenceValue=false
     bUseDistance2D=true
     boolValue=true
     clampMaxType=None
     clampMinType=None
     filterType=Minimum
     floatValueMax=0
     floatValueMin=0
     locationProviderContext=/Script/FortniteAIServer.FortQueryContext_AllPlayers
     multipleContextFilterOp=AllPass
     multipleContextScoreOp=AverageScore
     normalizationType=Absolute
     referenceValue=0
     scoreClampMax=0
     scoreClampMin=0
     scoringEquation=Linear
     scoringFactor=-1
     testComment=""
     testPurpose=FilterAndScore
    }
   }
  ]
 }]
